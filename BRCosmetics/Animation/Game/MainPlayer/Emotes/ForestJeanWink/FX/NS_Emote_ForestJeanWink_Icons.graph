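# Niagara system for the ForestJeanWink emote, producing cyan and pink lights, smoke and twinkling star particles.
# Lights_Loop002 drives mesh and sprite renderers from five spawn groups, switching lifetime, size, velocity, drag and noise per group, with smoke tinted by distance.
# Lights_Loop003 adds a pulsing cyan-to-pink light, dimmed to zero in the front end, plus two decals, one time-of-day tinted.
# Horse and HorseShoe emitters list no renderers.
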NiagaraSystem /BRCosmetics/Animation/Game/MainPlayer/Emotes/ForestJeanWink/FX/NS_Emote_ForestJeanWink_Icons
dependencies {
 usedDataInterfaces [&NiagaraDataInterfaceCurve, &NiagaraDataInterfaceVector2DCurve, &NiagaraDataInterfaceColorCurve]
 usedDynamicInputs [
  &MakeFloatFromVector2d
  &MakeCustomFloatFromBool
  &SetBoolByFloatComparison
  &MakeIntFromFloat
 ]
 usedModules [
  &self:SystemSpawnScript.SystemScriptSource.SystemScriptGraph.NiagaraNodeAssignment_1.SetVariables_#1
  &self:SystemSpawnScript.SystemScriptSource.SystemScriptGraph.NiagaraNodeAssignment_2.SetVariables_#2
  &SystemState
  &self:SystemSpawnScript.SystemScriptSource.SystemScriptGraph.NiagaraNodeAssignment_0.SetVariables_#3
 ]
}
emitters {
 Horse {rendererClasses [], renderers []}
 HorseShoe {rendererClasses [], renderers []}
 HorseShoe001 {rendererClasses [], renderers []}
 Lights_Loop002 {
  emitterUpdateScript [
   &EmitterState(
    Inactive Response=::NewEnumerator1
    Life Cycle Mode=::NewEnumerator1
    Loop Behavior=::NewEnumerator1
    Loop Duration Mode=::NewEnumerator1
    UseLoopDelay=false
   )
   &SpawnBurst_Instantaneous(
    Spawn Count=&RandomRangeInt(Minimum=10, Maximum=15)
    Spawn Group=4
    Use Spawn Probability=false
    UseLoopCountLimit=false
   ) {module=SpawnBurst_Instantaneous003}
   &SpawnRate(SpawnGroup=4, SpawnRate=25, Use Spawn Probability=false) {module=SpawnRate004}
   &SpawnBurst_Instantaneous(
    Spawn Count=&RandomRangeInt(Minimum=12, Maximum=16)
    Spawn Group=2
    Use Spawn Probability=false
    UseLoopCountLimit=false
   ) {module=SpawnBurst_Instantaneous002}
   &SpawnRate(SpawnGroup=2, SpawnRate=4, Use Spawn Probability=false) {module=SpawnRate002}
   &SpawnRate(SpawnGroup=3, SpawnRate=36, Use Spawn Probability=false) {module=SpawnRate003}
   &SpawnBurst_Instantaneous(Spawn Count=8, Spawn Group=1, Use Spawn Probability=false, UseLoopCountLimit=false) {module=SpawnBurst_Instantaneous001}
   &SpawnRate(SpawnGroup=1, SpawnRate=4, Use Spawn Probability=false) {module=SpawnRate001}
   &SpawnBurst_Instantaneous(Spawn Count=16, Spawn Time=0.12, Use Spawn Probability=false, UseLoopCountLimit=false)
   &SpawnRate(SpawnRate=36, Use Spawn Probability=false)
  ]
  particleSpawnScript [
   /SetVariables(Particles.VisibilityTag=&SpawnGroup()) {isSetParameters=true}
   &InitializeParticle(
    Color=&MakeCustomLinearColorFromBool(Bool=&RandomBool(), True Color=<rgba 0 1.5 1.4628 0.6522>, False Color=<rgba 1 0 0.7346 1>)
    Lifetime=&MergeHelper_FloatSwitch(
     0=&RandomRangeFloat(Minimum=1.3, Maximum=2.1)
     1=&RandomRangeFloat(Minimum=0.6, Maximum=1.5)
     2=&RandomRangeFloat(Minimum=1, Maximum=1.6)
     3=&RandomRangeFloat(Minimum=1, Maximum=0.5876)
     4=&RandomRangeFloat(Minimum=1.677, Maximum=1.8)
     Index=&SpawnGroup()
     Num=::NewEnumerator3
    )
    Mesh Scale Max=<3.5 3.5 5>
    Mesh Scale Min=<1.7 1.7 2>
    Mesh Scale Mode=::NewEnumerator2
    Position Offset=&MergeHelper_VectorSwitch(0=<0 -8 12>, 1=<0 0 0.6684>, 2=<0 0 0>, 3=<0 0 0>, 4=<0 -8 35>, Index=&SpawnGroup(), Num=::NewEnumerator3)
    Sprite Size=&MergeHelper_Vector2Switch(
     0=<87.0337 79.3783>
     1=&Multiply_Vector2D_ByFloat(Float=0.5, Vector2D=&UniformRangedVector2D(Minimum=<45 100>, Maximum=<80 165>))
     2=<0 0>
     3=<0 0>
     4=<0 0>
     Index=&SpawnGroup()
     Num=::NewEnumerator3
    )
    Sprite Size Mode=::NewEnumerator4
    UsePositionOffset=true
    Write Material Random=true
   )
   &InitialMeshOrientation(
    Mesh Orientation Mode=::NewEnumerator2
    Rotation=&MakeVector(
     X=&RandomRangeFloat(Minimum=-0.025, Maximum=0.1)
     Y=&RandomRangeFloat(Minimum=-0.25, Maximum=0.25)
     Z=0
    )
    Use Rotation Vector=true
   )
   &ShapeLocation(
    Apply To Particle Position=::NewEnumerator1
    Hemisphere Distribution=<0.036 1>
    MaskSpawnGroup=true
    Override Seed=false
    Rotation Angle=0.506
    Rotation Mode=::NewEnumerator1
    Spawn Group Mask=1
    Sphere Radius=55.7632
    UseOwnerScale=true
   )
   &ShapeLocation(
    Box Size=<120 20 100>
    MaskSpawnGroup=true
    Override Seed=false
    Shape Primitive=::NewEnumerator2
    Spawn Group Mask=4
    Surface Only=false
    UseOwnerScale=true
   ) {module=ShapeLocation001}
   &AddVelocity(
    Velocity=$Particles.ShapeLocation.ShapeNormal
    Velocity Speed Scale=&MergeHelper_FloatSwitch(
     0=0
     1=&RandomRangeFloat(Minimum=300, Maximum=515)
     2=&RandomRangeFloat(Minimum=300, Maximum=515)
     3=&RandomRangeFloat(Minimum=300, Maximum=515)
     4=&RandomRangeFloat(Minimum=20, Maximum=60)
     Index=&SpawnGroup()
     Num=::NewEnumerator3
    )
   )
  ]
  particleUpdateScript [
   &ParticleState()
   &ScaleSpriteSize(
    Scale Factor=&Vector2DFromCurve(
     Vector2Curve=DataInterface() {propertyValues {
       XCurve {keys [
         {arriveTangent=0.4813, interpMode=RCIM_Cubic, leaveTangent=0.4813, tangentMode=RCTM_User, time=0, value=0.625}
         {arriveTangent=1.25, interpMode=RCIM_Cubic, leaveTangent=1.25, time=0.4, value=0.9716}
         {arriveTangent=2.6138, interpMode=RCIM_Cubic, leaveTangent=2.6138, tangentMode=RCTM_User, time=0.9, value=1.75}
        ]}
       YCurve {keys [
         {arriveTangent=9.4573, interpMode=RCIM_Cubic, leaveTangent=9.4573, tangentMode=RCTM_User, time=0, value=0}
         {arriveTangent=0.8193, interpMode=RCIM_Cubic, leaveTangent=0.8193, tangentMode=RCTM_User, time=0.2, value=1.5019}
         <at 1 1.7992 RCIM_Cubic>
        ]}
      }}
     ScaleCurve=&Multiply_Vector2D_ByFloat(Float=&RandomRangeFloat(Minimum=0.7, Maximum=0.7), Vector2D=<0.8 1.5>)
    )
    Scale Sprite Size Mode=::NewEnumerator2
   )
   &Drag(
    Drag=&FloatFromCurve(
     FloatCurve=DataInterface() {propertyValues {Curve {
        defaultValue=1
        keys [
         {arriveTangent=4.1785, interpMode=RCIM_Cubic, leaveTangent=4.1785, tangentMode=RCTM_User, time=0, value=0}
         {arriveTangent=1.1179, interpMode=RCIM_Cubic, leaveTangent=1.1179, tangentMode=RCTM_User, time=0.2042, value=0.6228}
         {arriveTangent=0.0224, interpMode=RCIM_Cubic, leaveTangent=0.0224, tangentMode=RCTM_User, time=1, value=1}
        ]
       }}}
     Scale Curve=&MergeHelper_FloatSwitch(0=0, 1=26, 2=7.0767, 3=5.2844, 4=2, Index=&SpawnGroup(), Num=::NewEnumerator3)
    )
    Use Linear Drag=true
    Use Rotational Drag=false
   )
   &ScaleMeshSize(Scale Factor=&VectorFromFloat(Value=&FloatFromCurve(FloatCurve=DataInterface() {propertyValues {Curve {
         defaultValue=0
         keys [
          {arriveTangent=4.1785, interpMode=RCIM_Cubic, leaveTangent=4.1785, tangentMode=RCTM_User, time=0, value=0}
          {arriveTangent=1.1179, interpMode=RCIM_Cubic, leaveTangent=1.1179, tangentMode=RCTM_User, time=0.2042, value=0.6228}
          {arriveTangent=0.0224, interpMode=RCIM_Cubic, leaveTangent=0.0224, tangentMode=RCTM_User, time=1, value=1}
         ]
        }}})))
   /SetVariables(Particles.DynParamLightRay1:Vector4f=<0.2 0 0 0>) {isSetParameters=true}
   &DynamicMaterialParameters(
    Index 0 Param 1=&Multiply_Float(A=10, B=1)
    Index 0 Param 4=&FloatFromCurve(FloatCurve=DataInterface() {propertyValues {Curve {keys [{interpMode=RCIM_Cubic, tangentMode=RCTM_User, time=0, value=0}, {interpMode=RCIM_Cubic, tangentMode=RCTM_User, time=0.4, value=1}]}}})
    Index 1 Param 1=&FloatFromCurve(
     FloatCurve=DataInterface() {propertyValues {Curve {keys [
         {interpMode=RCIM_Cubic, tangentMode=RCTM_User, time=0, value=0}
         {arriveTangent=0.6295, interpMode=RCIM_Cubic, leaveTangent=0.6295, tangentMode=RCTM_User, time=1, value=1}
        ]}}}
     Scale Curve=0.42
    )
    Index 1 Param 2=&FloatFromCurve(
     FloatCurve=DataInterface() {propertyValues {Curve {keys [
         {arriveTangent=-100.8588, time=0, value=1}
         {arriveTangent=-26.7969, time=0.2, value=0.1}
         {arriveTangent=-2.6967, leaveTangent=-2.6967, time=0.3, value=0}
        ]}}}
     CurveIndex=$Emitter.Age
     Scale Curve=&MaskFloatBySpawnGroup(Masked Float=0, Passthrough Float=5, Spawn Group Mask=1)
    )
    Index 1 Param 3=&FloatFromCurve(
     FloatCurve=DataInterface() {propertyValues {Curve {keys [{interpMode=RCIM_Cubic, tangentMode=RCTM_User, time=0.2, value=0.5}, {interpMode=RCIM_Cubic, tangentMode=RCTM_User, time=1, value=1}]}}}
     Scale Curve=2
    )
    Index 1 Param 4=&FloatFromCurve(FloatCurve=DataInterface() {propertyValues {Curve {keys [<at 0 1>,<at 1 0>]}}}, Scale Curve=-0.3)
    Module.Param0WriteEnabled=true
    Module.Param0WriteEnabled1=true
    Module.Param1WriteEnabled=false
    Module.Param1WriteEnabled1=true
    Module.Param2WriteEnabled=false
    Module.Param2WriteEnabled1=true
    Module.Param3WriteEnabled=true
    Module.Param3WriteEnabled1=true
    Write Parameter Index 1=true
   )
   &ScaleColor(
    Linear Color Curve=DataInterface() {propertyValues {
      AlphaCurve {keys [<at 0 0>,<at 0.2 1>,<at 1 0>]}
      BlueCurve {keys [<at 0 1>,<at 1 1>]}
      GreenCurve {keys [<at 0 1>,<at 1 1>]}
      RedCurve {keys [<at 0 1>,<at 1 1>]}
     }}
    Scale Mode=::NewEnumerator2
   ) {module=ScaleColor001}
   /SetVariables() {isSetParameters=true}
   &CurlNoiseForce(
    Bias Noise=true
    Bias Noise Field:Vector3f=<1 0 1>
    Noise Frequency=12
    Noise Strength=&MergeHelper_FloatSwitch(
     0=0
     1=0
     2=&RandomRangeFloat(Minimum=60, Maximum=175)
     3=&RandomRangeFloat(Minimum=60, Maximum=175)
     4=0
     Index=&SpawnGroup()
     Num=::NewEnumerator3
    )
    Pan Noise=false
   )
   &SolveForcesAndVelocity(Clamp Velocity=false, Limit Acceleration=false)
   /SetVariables(
    Particles.MotesSize:Vector2f=&Vector2DFromCurve(
     Vector2Curve=DataInterface() {propertyValues {XCurve {keys [<at 0 1>,<at 1 1>]}, YCurve {keys [<at 0 0.3 RCIM_Cubic>,<at 1 1.5 RCIM_Cubic>]}}}
     ScaleCurve=&RandomRangeVector2D(Minimum=<2.614 78.0441>, Maximum=<5.3015 56.4988>)
    )
    Particles.StarColor:LinearColor=&Multiply_LinearColorByFloat(
     Linear Color=<rgba 6 3.3677 0 1>
     Float=&Multiply_Float(
      A=&FloatFromCurve(FloatCurve=DataInterface() {propertyValues {Curve {
          defaultValue=1
          keys [
           {arriveTangent=9.0725, interpMode=RCIM_Cubic, leaveTangent=9.0725, tangentMode=RCTM_User, time=0, value=0}
           {arriveTangent=-2.5249, interpMode=RCIM_Cubic, leaveTangent=-2.5249, tangentMode=RCTM_User, time=0.1408, value=0.9847}
           {arriveTangent=-0.0405, interpMode=RCIM_Cubic, leaveTangent=-0.0405, tangentMode=RCTM_User, time=1, value=0}
          ]
         }}})
      B=&Waveform(
       "Waveform [1]"=::NewEnumerator0
       "[1] Frequency"=&FloatFromCurve(
        FloatCurve=DataInterface() {propertyValues {Curve {keys [<at 0.2 0>,<at 0.4 1>]}}}
        Scale Curve=&RandomRangeFloat(Minimum=3, Maximum=5)
       )
       "[1] Amplitude Min/Max"=<0 1>
      )
     )
    )
    Particles.StarRot=45
    Particles.StarSize:Vector2f=&Multiply_Vector2D_ByFloat(
     Float=&FloatFromCurve(FloatCurve=DataInterface() {propertyValues {Curve {
         defaultValue=1
         keys [
          {arriveTangent=9.0725, interpMode=RCIM_Cubic, leaveTangent=9.0725, tangentMode=RCTM_User, time=0, value=0}
          {arriveTangent=-2.5249, interpMode=RCIM_Cubic, leaveTangent=-2.5249, tangentMode=RCTM_User, time=0.1408, value=0.9847}
          {arriveTangent=-0.0405, interpMode=RCIM_Cubic, leaveTangent=-0.0405, tangentMode=RCTM_User, time=1, value=0}
         ]
        }}})
     Vector2D=<15 15>
    )
   ) {isSetParameters=true}
   /SetVariables(
    Particles.SmokeColor:LinearColor=&Multiply_LinearColorByFloat(
     Linear Color=&ColorFromCurve(
      ColorCurve=DataInterface() {propertyValues {
        AlphaCurve {keys [<at 0 0.7 RCIM_Cubic>]}
        BlueCurve {keys [<at 0.0758 1 RCIM_Cubic>,<at 0.2428 0.7233>,<at 0.4226 0.9197 RCIM_Cubic>]}
        GreenCurve {keys [<at 0.0758 0.9488 RCIM_Cubic>,<at 0.2428 0.3217>,<at 0.4226 0.0816 RCIM_Cubic>]}
        RedCurve {keys [<at 0.0758 0 RCIM_Cubic>,<at 0.2428 0.1852>,<at 0.4226 1 RCIM_Cubic>]}
       }}
      CurveIndex=&NormalizeDistanceRange(
       Start Position:NiagaraPosition=<0 -3 35>
       End Position:NiagaraPosition=$Particles.Position
       Distance:NiagaraFloat=160
       Invert Normalized Range:NiagaraBool=false
      )
     )
     Float=&FloatFromCurve(FloatCurve=DataInterface() {propertyValues {Curve {
         defaultValue=1
         keys [
          {interpMode=RCIM_Cubic, tangentMode=RCTM_User, time=0, value=0}
          {interpMode=RCIM_Cubic, tangentMode=RCTM_User, time=0.5555, value=0.55}
          {tangentMode=RCTM_User, time=1, value=0}
         ]
        }}})
    )
    Particles.SmokeDynParam1:Vector4f=&MakeVector4(X=&RandomRangeFloat(Minimum=0, Maximum=1), Y=0, Z=0, W=0)
    Particles.SmokeDynParam3:Vector4f=&MakeVector4(
     X=1
     Y=&RandomRangeFloat(Minimum=0, Maximum=1)
     Z=&FloatFromCurve(FloatCurve=DataInterface() {propertyValues {Curve {keys [
          {interpMode=RCIM_Cubic, leaveTangent=2.0509, tangentMode=RCTM_Break, time=0, value=0}
          {arriveTangent=0.1748, interpMode=RCIM_Cubic, tangentMode=RCTM_Break, time=1, value=1}
         ]}}})
     W=0.75
    )
    Particles.SmokeSize:Vector2f=&Multiply_Vector2D_ByFloat(
     Float=&FloatFromCurve(
      FloatCurve=DataInterface() {propertyValues {Curve {
         defaultValue=1
         keys [
          {interpMode=RCIM_Cubic, tangentMode=RCTM_User, time=0, value=0.4}
          {interpMode=RCIM_Cubic, tangentMode=RCTM_User, time=0.7538, value=1}
          {arriveTangent=-5.7174, tangentMode=RCTM_User, time=1, value=0.4}
         ]
        }}}
      Scale Curve=&RandomRangeFloat(Minimum=120, Maximum=160)
     )
     Vector2D=<1 1>
    )
   ) {isSetParameters=true}
  ]
  rendererClasses [&NiagaraMeshRendererProperties, &NiagaraSpriteRendererProperties]
  renderers [
   {rendererClass=&NiagaraMeshRendererProperties, rendererIndex=0}
   {rendererClass=&NiagaraMeshRendererProperties, rendererIndex=1}
   {rendererClass=&NiagaraSpriteRendererProperties, rendererIndex=2}
   {rendererClass=&NiagaraSpriteRendererProperties, rendererIndex=3}
   {rendererClass=&NiagaraSpriteRendererProperties, rendererIndex=4}
   {rendererClass=&NiagaraSpriteRendererProperties, rendererIndex=5}
  ]
 }
 Lights_Loop003 {
  emitterUpdateScript [
   &EmitterState(Inactive Response=::NewEnumerator1, Life Cycle Mode=::NewEnumerator1, Loop Behavior=::NewEnumerator1, Loop Duration=7.8, UseLoopDelay=false)
   &SpawnBurst_Instantaneous(Spawn Time=0.12, Use Spawn Probability=false, UseLoopCountLimit=false)
  ]
  particleSpawnScript [&InitializeParticle(
    Color=<rgba 0.0075 0.9131 0.8388 1>
    Lifetime=7.68
    Mesh Scale Max=<3.5 3.5 5>
    Mesh Scale Min=<1.7 1.7 2>
    Mesh Scale Mode=::NewEnumerator2
    Position Offset=<0 -8 12>
    Sprite Size Mode=::NewEnumerator3
    UsePositionOffset=true
    Write Material Random=true
   )]
  particleUpdateScript [
   /SetVariables(Particles.DecalColor:LinearColor=&FNTimeOfDayColor(Night Color=<rgba 0 1 0.8278 1>, Morning Color=<rgba 0 1 0.8278 1>, Day Color=<rgba 0.3698 1 0.8915 1>, Evening Color=<rgba 0 1 0.8278 1>)) {isSetParameters=true}
   &ParticleState(Kill Particles When Lifetime Has Elapsed=false)
   &FN_Light_Attributes(
    Color=&LerpLinearColors(
     StartColor=<rgba 4.3992 19.0636 20 1>
     EndColor=<rgba 20 1.434 19.176 1>
     LerpFactor=&Waveform("Waveform [1]"=::NewEnumerator0, "[1] Frequency"=1, "[1] Amplitude Min/Max"=<0 1>)
    )
    Create Color Attribute=true
    Create Position Attribute=true
    Create Radius Attribute=true
    Radius=&Waveform("Waveform [1]"=::NewEnumerator0, "[1] Frequency"=1, "[1] Amplitude Min/Max"=<120 150>)
    Scale Color=&FloatFromCurve(
     FloatCurve=DataInterface() {propertyValues {Curve {keys [<at 0 0 RCIM_Cubic>,<at 0.2 1>]}}}
     CurveIndex=$Emitter.Age
     Scale Curve=&FN_FrontEndFloatModifier(Gameplay=0.6, Front End=0)
    )
   )
   /Niagara/Modules/Decal/Decal_Attributes(Decal Size:Vector3f=<300 300 300>, Size=2)
   /SetVariables(Particles.float=9999) {isSetParameters=true}
   /SetVariables(Particles.DecalColor2:LinearColor=<rgba 1 0.059 0.6741 1>, Particles.DecalSize2:Vector3f=<300 400 400>) {isSetParameters=true}
  ]
  rendererClasses [&NiagaraLightRendererProperties, &NiagaraDecalRendererProperties]
  renderers [
   {rendererClass=&NiagaraLightRendererProperties, rendererIndex=0}
   {rendererClass=&NiagaraDecalRendererProperties, rendererIndex=1}
   {rendererClass=&NiagaraDecalRendererProperties, rendererIndex=2}
  ]
 }
}
summary {
 systemName=NS_Emote_ForestJeanWink_Icons
 emitters [
  {emitterName=Horse, rendererClasses []}
  {emitterName=HorseShoe, rendererClasses []}
  {emitterName=HorseShoe001, rendererClasses []}
  {emitterName=Lights_Loop002, rendererClasses [&NiagaraMeshRendererProperties, &NiagaraSpriteRendererProperties]}
  {emitterName=Lights_Loop003, rendererClasses [&NiagaraLightRendererProperties, &NiagaraDecalRendererProperties]}
 ]
 userVariables [
  User.BeatSwitch {type=NiagaraFloat, defaultValue=0, description=""}
  User.bIsFrontEnd {type=NiagaraBool}
  User.bIsFrontEndPreview {type=NiagaraBool}
 ]
}
